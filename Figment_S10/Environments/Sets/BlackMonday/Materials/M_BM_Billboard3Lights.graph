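# Unlit, Nanite-enabled material for the lights on a Black Monday billboard, named for three lights.
# Three scalar masks are appended into a vector and dotted with vertex color, selecting or blending the red, green and blue painted regions.
# That mask multiplies an HDR emissive color, pale blue by default.
# Masks and color all read Custom Primitive Data, so each primitive can set its lit regions and color without separate material instances.
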Material /Figment_S10/Environments/Sets/BlackMonday/Materials/M_BM_Billboard3Lights
graph {
 properties {bUsedWithNanite=true, pixelDepthOffsetMode=PDOM_Legacy, shadingModel=MSM_Unlit, usedShadingModels=Unlit}
 nodes [
  n1 = VectorParameter() {bUseCustomPrimitiveData=true, parameterName=Emissive, primitiveDataIndex=28, defaultValue=<rgba 7.7742 9.7977 20 1>}
  n2 = ScalarParameter() {bUseCustomPrimitiveData=true, parameterName=VertexMaskRed, primitiveDataIndex=25}
  n3 = ScalarParameter() {bUseCustomPrimitiveData=true, parameterName=VertexMaskGreen, primitiveDataIndex=26}
  n4 = ScalarParameter() {bUseCustomPrimitiveData=true, parameterName=VertexMaskBlue, primitiveDataIndex=27}
  Multiply(
   A=@n1.RGB
   B=DotProduct(
    A=MaterialFunctionCall(R=@n2, G=@n3, B=@n4) {materialFunction=&AppendMany}.RGB
    B=VertexColor()
   )
  )
 ]
}
