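# Translucent deferred decal material for the BlackMonday environment set, with parameters for texture, normal, roughness, emissive and opacity.
# The Base Color texture's RGB drives an emissive-scaled term and its alpha an opacity term, both masked by MF_DecalFacingMask borrowed from the SprayGun prototype.
# A scalar also named Base Color multiplies the texture's alpha rather than its RGB.
# The normal map is scaled by plain multiplication with NormalIntensity.
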Material /Figment_S10/Environments/Sets/BlackMonday/Materials/M_BM_Decals
graph {
 properties {blendMode=BLEND_Translucent, materialDomain=MD_DeferredDecal, pixelDepthOffsetMode=PDOM_Legacy, usedShadingModels=Default Lit}
 nodes [
  Constant()
  Constant() {r=0.5}
  n1 = TextureSampleParameter2D() {parameterName=Base Color}
  Multiply(A=@n1.A, B=ScalarParameter() {defaultValue=0.5, parameterName=Base Color})
  Multiply(
   A=TextureSampleParameter2D() {parameterName=Normal, samplerType=SAMPLERTYPE_Normal, texture=&FlatNormalBlue}.RGB
   B=ScalarParameter() {defaultValue=1, parameterName=NormalIntensity}
  )
  n2 = MaterialFunctionCall() {materialFunction=/Game/Weapons/Prototype/SprayGun/Materials/MF_DecalFacingMask}
  Multiply(A=@n2.Result, B=Multiply(A=@n1.RGB, B=ScalarParameter() {defaultValue=0.3, parameterName=Emmissive Color}))
  Multiply(A=@n2.Result, B=Multiply(A=@n1.A, B=ScalarParameter() {defaultValue=1, parameterName=Opacity}))
  ScalarParameter() {defaultValue=1, parameterName=Roughness}
 ]
}
